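# Animation Blueprint for the Frozen Reality pickaxe prop, named for unequip, that switches the axe from an opened pose to a closed pose.
# A two-state machine starts on the opened pose and moves to the non-looping closed pose once Is Builtin Emote is true.
# The event graph sets that flag unconditionally after a short delay from initialization.
# The result passes through DefaultSlot, then an additive bone-space translation offsets b_AR_Root.
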AnimBlueprint /BRCosmetics/Weapons/FORT_Melee/Pickaxe_FrozenReality/Animation/FrozenReality_Axe_Prop_Unequip_AnimBP
target_skeleton=/BRCosmetics/Weapons/FORT_Melee/Pickaxe_FrozenReality/Meshes/FrozenReality_Axe_Skeleton
graphs [
 AnimGraph {nodes [
   AnimGraphNode_StateMachine_0 = Idle Pose() {is=title AnimGraphNode_StateMachine}
   n1 = "Transform (Modify) Bone - Bone: b_AR_Root"(
    ComponentPose=Local To Component(LocalPose=Slot 'DefaultSlot'(Source=@AnimGraphNode_StateMachine_0))
    Alpha=1
    Translation="0,-4,0"
    Rotation="0,0,0"
    Scale="1,1,1"
   ) {
    translationMode=BMM_Additive
    translationSpace=BCS_BoneSpace
    alphaBoolBlend {blendOption=Linear}
    boneToModify {boneName=b_AR_Root}
    translation {y=-4}
   }
   Output Pose(Result=Component To Local(ComponentPose=@n1)) {layerGroup=None}
  ]}
 Idle Pose {
  entry=Idle
  path=AnimGraph.AnimGraphNode_StateMachine_0.Idle Pose
  ids=id
  nodes [
   AnimStateNode_0 = Idle()
   AnimStateNode_1 = FinalState() {is=title AnimStateNode}
   AnimStateTransitionNode_0 = AnimStateTransitionNode()
  ]
  links [{
    dst=FinalState
    rule=AnimGraph.AnimGraphNode_StateMachine_0.Idle Pose.AnimStateTransitionNode_0.Transition
    src=Idle
    via=AnimStateTransitionNode_0
   }]
 }
 Idle {
  path=AnimGraph.AnimGraphNode_StateMachine_0.Idle Pose.AnimStateNode_0.Idle
  nodes [
   Sequence Player 'FrozenReality_Pickaxe_Idle_WEP'() {sequence=/BRCosmetics/Animation/Game/MainPlayer/Combat/Melee/OneHanded/FrozenReality/FrozenReality_Pickaxe_Idle_WEP}
   Output Animation Pose(Result=Sequence Player 'Emote_FrozenReality_OpenedPose'() {sequence=&/Emote_FrozenReality_OpenedPose})
  ]
 }
 FinalState {
  path=AnimGraph.AnimGraphNode_StateMachine_0.Idle Pose.AnimStateNode_1.FinalState
  nodes [Output Animation Pose(Result=Sequence Player 'Emote_FrozenReality_ClosedPose'() {bLoopAnimation=false, sequence=&/Emote_FrozenReality_ClosedPose})]
 }
 Transition {
  path=AnimGraph.AnimGraphNode_StateMachine_0.Idle Pose.AnimStateTransitionNode_0.Transition
  nodes [Result(bCanEnterTransition=Get Is Builtin Emote() -> (Is Builtin Emote:bool))]
 }
 EventGraph {nodes [
   Event BlueprintInitializeAnimation()
   > Delay(Duration=1)
   > Set Is Builtin Emote(Is Builtin Emote:bool=true) -> (Output_Get:bool)
  ]}
]
variables [Is Builtin Emote {type=boolean}]
