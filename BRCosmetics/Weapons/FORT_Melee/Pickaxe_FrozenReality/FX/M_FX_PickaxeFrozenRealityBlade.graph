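# Additive unlit material for Niagara mesh particles, an effect on the Frozen Reality pickaxe blade.
# Particle color feeds an emissive adjustments function.
# A soft sphere mask between UV0's vertical coordinate and dynamic RevealT is compared with scrolling UV1 flame noise in a second sphere mask, giving a noise-broken reveal scaled by particle alpha.
# A lerp, a spare panner and parameter T are unconnected.
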Material /BRCosmetics/Weapons/FORT_Melee/Pickaxe_FrozenReality/FX/M_FX_PickaxeFrozenRealityBlade
graph {
 properties {
  bUsedWithNiagaraMeshParticles=true
  blendMode=BLEND_Additive
  pixelDepthOffsetMode=PDOM_Legacy
  shadingModel=MSM_Unlit
  usedShadingModels=Unlit
 }
 nodes [
  LinearInterpolate()
  n1 = ParticleColor()
  MaterialFunctionCall(Emissive=@n1.RGB) {materialFunction=&MF_FX_EmissiveAdjustments}
  n2 = TextureCoordinate() {coordinateIndex=1}
  n3 = Panner(Coordinate=Multiply(A=ComponentMask(None=@n2) {g=true, r=true}, B=ScalarParameter() {defaultValue=0.75, parameterName=Scale})) {bFractionalPart=true, speedY=0.1}
  n4 = ComponentMask(None=TextureCoordinate()) {g=true}
  n5 = DynamicParameter() {paramNames [RevealT,Param2,Param3,Param4]}
  Multiply(
   A=@n1.A
   B=SphereMask(
    A=TextureSampleParameter2D(UVs=@n3) {parameterName=T_Noise, texture=/Game/Effects/FogVolume/noise_flamy_Tile}.R
    B=SphereMask(A=@n4, B=@n5.RevealT) {attenuationRadius=2, hardnessPercent=0}
    Radius=ScalarParameter() {parameterName=r}
   ) {hardnessPercent=10}
  )
  Panner()
  ScalarParameter() {parameterName=T}
 ]
}
