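# Additive unlit glow material for Niagara mesh particles, part of the Frozen Reality pickaxe effects.
# A normalized, transformed GrowDirection vector is scaled by U and the GlowWidth dynamic parameter, forming what looks like a vertex-stretch offset.
# The glow mask combines stretched, panning noise with a steep falloff along U and a soft band centered in V, scaled by particle alpha.
# The LightWidth parameter is unconnected.
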Material /BRCosmetics/Weapons/FORT_Melee/Pickaxe_FrozenReality/FX/M_PickaxeFrozenReality_Glow
graph {
 properties {
  bUsedWithNiagaraMeshParticles=true
  blendMode=BLEND_Additive
  pixelDepthOffsetMode=PDOM_Legacy
  shadingModel=MSM_Unlit
  usedShadingModels=Unlit
 }
 nodes [
  n1 = Normalize(VectorInput=VectorParameter() {parameterName=GrowDirection, defaultValue=<rgba 0 0 0 1>}.RGB)
  n2 = TextureCoordinate()
  Multiply(
   A=Multiply(A=Transform(None=@n1), B=ComponentMask(None=@n2) {r=true})
   B=DynamicParameter() {paramNames [GlowWidth,Param2,Param3,Param4]}.GlowWidth
  )
  n3 = TextureSample(UVs=Panner(Coordinate=TextureCoordinate() {uTiling=0.075, vTiling=2.25}) {speedX=-0.2}) {samplerType=SAMPLERTYPE_LinearGrayscale, texture=/Game/Effects/Fort_Effects/Emotes/UltraLight/Texture/T_Tiling_Noise_01_GreyScale}
  n4 = TextureCoordinate()
  n5 = Power(Base=OneMinus(None=ComponentMask(None=@n4) {r=true})) {constExponent=4}
  n6 = ComponentMask(None=TextureCoordinate()) {g=true}
  n7 = SphereMask(A=Fmod(A=Abs(None=@n6), B=Constant() {r=1}), B=Constant() {r=0.5}) {attenuationRadius=0.5, hardnessPercent=15}
  Multiply(A=ParticleColor().A, B=Multiply(A=Add(A=@n3.R) {constB=0.35}, B=Multiply(A=@n5, B=@n7)))
  ScalarParameter() {parameterName=LightWidth}
 ]
}
