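# Animation Blueprint for the Frozen Reality pickaxe mesh, switching its pose on whether an emote is playing.
# During emotes it passes the input pose through; otherwise a single-state machine supplies the Emote_FrozenReality_OpenedPose sequence.
# The blend feeds DefaultSlot for montages, then a constant additive bone-space offset shifts b_AR_Root along negative Y.
# The event graph is empty, and the weapon idle sequence player is left unconnected.
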AnimBlueprint /BRCosmetics/Weapons/FORT_Melee/Pickaxe_FrozenReality/Meshes/FrozenReality_Axe_AnimBP
target_skeleton=&/FrozenReality_Axe_Skeleton
graphs [
 AnimGraph {nodes [
   AnimGraphNode_StateMachine_0 = Idle Pose() {is=title AnimGraphNode_StateMachine}
   AnimGraphNode_SubInput_0 = Input Pose()
   n1 = Blend Poses by bool(BlendPose_0=@AnimGraphNode_SubInput_0, BlendPose_1=@AnimGraphNode_StateMachine_0, bActiveValue=Get bIsPlayingEmote()) {blendType=Linear}
   n2 = "Transform (Modify) Bone - Bone: b_AR_Root"(ComponentPose=Local To Component(LocalPose=Slot 'DefaultSlot'(Source=@n1)), Alpha=1, Translation="0,-4,0", Rotation="0,0,0", Scale="1,1,1") {
    translationMode=BMM_Additive
    translationSpace=BCS_BoneSpace
    alphaBoolBlend {blendOption=Linear}
    boneToModify {boneName=b_AR_Root}
    translation {y=-4}
   }
   Output Pose(Result=Component To Local(ComponentPose=@n2)) {layerGroup=None}
  ]}
 Idle Pose {entry=Idle, path=AnimGraph.AnimGraphNode_StateMachine_0.Idle Pose, ids=id, nodes [AnimStateNode_0 = Idle()]}
 Idle {
  path=AnimGraph.AnimGraphNode_StateMachine_0.Idle Pose.AnimStateNode_0.Idle
  nodes [
   Sequence Player 'FrozenReality_Pickaxe_Idle_WEP'() {sequence=/BRCosmetics/Animation/Game/MainPlayer/Combat/Melee/OneHanded/FrozenReality/FrozenReality_Pickaxe_Idle_WEP}
   Output Animation Pose(Result=Sequence Player 'Emote_FrozenReality_OpenedPose'() {sequence=/BRCosmetics/Weapons/FORT_Melee/Pickaxe_FrozenReality/Animation/Emote_FrozenReality_OpenedPose})
  ]
 }
 EventGraph {links [], nodes []}
]
variables [
 BuiltIn Emote {type=number}
 Is Builtin Emote {type=boolean}
]
